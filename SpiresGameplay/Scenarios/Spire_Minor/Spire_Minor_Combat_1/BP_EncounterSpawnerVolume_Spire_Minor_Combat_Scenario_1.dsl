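; Encounter spawner volume Blueprint for Minor Spire Combat Scenario 1 that gates when its encounter spawning is active.
; The construction script disables spawning if a Minor Spire point provider is linked and enables it otherwise.
; On BeginPlay, a disabled volume binds to the provider's OnActorSpawned event, which fires when LWM selects it to spawn.
; The first spawned actor unbinds that handler and enables encounter spawning.
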(event EventBeginPlay
  (|CallParentFunction)
  ; If a Minor Spire point provider is linked, only enable the volume when it has been selected to spawn by the LWM.
  (if (and (not (CallFunction|IsEncounterSpawningEnabled)) (Utilities|IsValid (Variables|Encounter|GetMinor%20Spire%20Point%20Provider) :!class /Script/Engine.KismetSystemLibrary))
    (EventDispatchers|AssignOnActorSpawned (Variables|Encounter|GetMinor%20Spire%20Point%20Provider) (delegate Custom|PointProvider_OnActorSpawned))))

(event Custom|PointProvider_OnActorSpawned
  (params (SpawnedActor :type Actor%20Object%20Reference))
  ; If a Minor Spire point provider is linked, only enable the volume when it has been selected to spawn by the LWM.
  (EventDispatchers|UnbindEventfromOnActorSpawned (Variables|Encounter|GetMinor%20Spire%20Point%20Provider) (delegate Custom|PointProvider_OnActorSpawned)) ; Only enable for the first spawned actor.
  (CallFunction|EnableEncounterSpawning))

(fn ConstructionScript
  ;
  (Utilities|IsValid (Variables|Encounter|GetMinor%20Spire%20Point%20Provider)
    (:Is%20Valid
      (CallFunction|DisableEncounterSpawning))
    (:Is%20Not%20Valid
      (CallFunction|EnableEncounterSpawning))))
