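# StateTree for the Spire_Minor_Combat_1 quest scenario, a combat encounter around a minor Spire.
# Global tasks read unit queries, combatant count and the Spire actor.
# Init grants a Spire reaction and waits for players, then a count of 3 or 4 selects the Spires-faction wave spawned at tagged points.
# Spawns then turn CherrySmoke-hostile; the wave-remaining transition to Spire_Vulnerable and the wave quest grants are disabled.
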StateTree /SpiresGameplay/Scenarios/Spire_Minor/Spire_Minor_Combat_1/ST_QuestScenario_Spire_Minor_Combat_1
global_tasks [
 {
  id=#1
  description="Get Variable [Tag Query] - UnitQuery1"
  properties {
   iD=#2
   instance {
    _structType=/Script/EncountersRuntime.EncounterGetVariableTagQueryStateTreeTaskInstanceData
    userActor=null
    value {autoDescription="", tokenStreamVersion=0, userDescription="", queryTokenStream [], tagDictionary []}
    variableIdentifier {tagName=QuestEncounter.Variables.UnitQuery1}
   }
   node {
    name="Get Variable [Tag Query] - UnitQuery1"
    _structType=/Script/EncountersRuntime.EncounterGetVariableTagQueryStateTreeTask
    bCanEditConsideredForCompletion=true
    bConsideredForCompletion=true
    bHasTaskCompletionDelegateDispatcher=false
    bTaskEnabled=true
    transitionHandlingPriority=Normal
    bindingsBatch {value=65535}
    executionRuntimeTemplateIndex {value=65535}
    instanceDataHandle {index=65535, source=None, stateHandle {index=65535}}
    instanceTemplateIndex {value=65535}
    outputBindingsBatch {value=65535}
   }
  }
 }
 {
  id=#3
  description="Get Variable [Tag Query] - UnitQuery2"
  properties {
   iD=#4
   instance {
    _structType=&EncounterGetVariableTagQueryStateTreeTaskInstanceData
    userActor=null
    value {autoDescription="", tokenStreamVersion=0, userDescription="", queryTokenStream [], tagDictionary []}
    variableIdentifier {tagName=QuestEncounter.Variables.UnitQuery2}
   }
   node {
    name="Get Variable [Tag Query] - UnitQuery2"
    _structType=&EncounterGetVariableTagQueryStateTreeTask
    bCanEditConsideredForCompletion=true
    bConsideredForCompletion=true
    bHasTaskCompletionDelegateDispatcher=false
    bTaskEnabled=true
    transitionHandlingPriority=Normal
    bindingsBatch {value=65535}
    executionRuntimeTemplateIndex {value=65535}
    instanceDataHandle {index=65535, source=None, stateHandle {index=65535}}
    instanceTemplateIndex {value=65535}
    outputBindingsBatch {value=65535}
   }
  }
 }
 {
  id=#5
  description="Get Variable [Tag Query] - UnitQuery3"
  properties {
   iD=#6
   instance {
    _structType=&EncounterGetVariableTagQueryStateTreeTaskInstanceData
    userActor=null
    value {autoDescription="", tokenStreamVersion=0, userDescription="", queryTokenStream [], tagDictionary []}
    variableIdentifier {tagName=QuestEncounter.Variables.UnitQuery3}
   }
   node {
    name="Get Variable [Tag Query] - UnitQuery3"
    _structType=&EncounterGetVariableTagQueryStateTreeTask
    bCanEditConsideredForCompletion=true
    bConsideredForCompletion=true
    bHasTaskCompletionDelegateDispatcher=false
    bTaskEnabled=true
    transitionHandlingPriority=Normal
    bindingsBatch {value=65535}
    executionRuntimeTemplateIndex {value=65535}
    instanceDataHandle {index=65535, source=None, stateHandle {index=65535}}
    instanceTemplateIndex {value=65535}
    outputBindingsBatch {value=65535}
   }
  }
 }
 {
  id=#7
  description="Get Variable [Tag Query] - UnitQuery4"
  properties {
   iD=#8
   instance {
    _structType=&EncounterGetVariableTagQueryStateTreeTaskInstanceData
    userActor=null
    value {autoDescription="", tokenStreamVersion=0, userDescription="", queryTokenStream [], tagDictionary []}
    variableIdentifier {tagName=QuestEncounter.Variables.UnitQuery4}
   }
   node {
    name="Get Variable [Tag Query] - UnitQuery4"
    _structType=&EncounterGetVariableTagQueryStateTreeTask
    bCanEditConsideredForCompletion=true
    bConsideredForCompletion=true
    bHasTaskCompletionDelegateDispatcher=false
    bTaskEnabled=true
    transitionHandlingPriority=Normal
    bindingsBatch {value=65535}
    executionRuntimeTemplateIndex {value=65535}
    instanceDataHandle {index=65535, source=None, stateHandle {index=65535}}
    instanceTemplateIndex {value=65535}
    outputBindingsBatch {value=65535}
   }
  }
 }
 {
  id=#9
  description="Get Variable [Int] - CombatantCount"
  properties {
   iD=#10
   instance {
    _structType=/Script/EncountersRuntime.EncounterGetVariableIntStateTreeTaskInstanceData
    userActor=null
    value=0
    variableIdentifier {tagName=QuestEncounter.Variables.CombatantCount}
   }
   node {
    name="Get Variable [Int] - CombatantCount"
    _structType=/Script/EncountersRuntime.EncounterGetVariableIntStateTreeTask
    bCanEditConsideredForCompletion=true
    bConsideredForCompletion=true
    bHasTaskCompletionDelegateDispatcher=false
    bTaskEnabled=true
    transitionHandlingPriority=Normal
    bindingsBatch {value=65535}
    executionRuntimeTemplateIndex {value=65535}
    instanceDataHandle {index=65535, source=None, stateHandle {index=65535}}
    instanceTemplateIndex {value=65535}
    outputBindingsBatch {value=65535}
   }
  }
 }
 {
  id=#11
  description=Get Scenario Actor - MinorSpire
  properties {
   iD=#12
   instance {
    _structType=&GetEncounterActorStateTreeTaskInstanceData
    actorQuery {
     autoDescription=" ANY( QuestEncounter.Spire_Minor_Random.Encounter.Target.Spire )"
     tagDictionary [{tagName=QuestEncounter.Spire_Minor_Random.Encounter.Target.Spire}]
    }
   }
   node {name=Get Scenario Actor - MinorSpire, _structType=&GetEncounterActorStateTreeTask}
  }
 }
]
states [
 Root {
  id=#13
  type=STATE
  children [
   #14
   #15
   #16
   #17
   #18
   #19
   #20
   #21
  ]
  tasks []
  transitions []
 }
 Init {
  id=#14
  type=STATE
  tasks [
   {
    id=#22
    description=Grant Reactions - Spire invulnerable
    properties {
     iD=#23
     instance {
      _structType=/Script/EncountersRuntime.GrantReactionStateTreeTaskInstanceData
      grantType=GrantToTargetActor
      targetActor=null
      userActor=null
      reactionEntries [{rowValue {
         bUseDataRegistryId=false
         dataRegistryId=""
         dataTableRow {
          dataTable=&/DT_QuestScenario_Spire_Minor_Combat_1_Reactions
          rowName=SpireVulnerable
         }
        }}]
      reactions []
     }
     node {
      name=Grant Reactions - Spire invulnerable
      _structType=/Script/EncountersRuntime.GrantReactionStateTreeTask
      bCanEditConsideredForCompletion=true
      bConsideredForCompletion=true
      bHasTaskCompletionDelegateDispatcher=false
      bTaskEnabled=true
      transitionHandlingPriority=Normal
      bindingsBatch {value=65535}
      executionRuntimeTemplateIndex {value=65535}
      instanceDataHandle {index=65535, source=None, stateHandle {index=65535}}
      instanceTemplateIndex {value=65535}
      outputBindingsBatch {value=65535}
     }
    }
   }
   {
    id=#24
    description="Scenario [Player Count Transition]"
    properties {
     iD=#25
     instance {
      _structType=/Script/EncountersRuntime.EncounterPlayerCountTransitionStateTreeTaskInstanceData
      bOnlyTriggerOnce=true
      bShouldTransition=true
      bTriggered=false
      comparisonOperator=GreaterThan
      userActor=null
      value=0
      customVerb {amount=1, customVerbTag {tagName=None}, subjects [], values []}
      onTriggered {iD=#26}
     }
     node {
      name=None
      _structType=/Script/EncountersRuntime.EncounterPlayerCountTransitionStateTreeTask
      bCanEditConsideredForCompletion=true
      bConsideredForCompletion=true
      bHasTaskCompletionDelegateDispatcher=false
      bTaskEnabled=true
      transitionHandlingPriority=Normal
      bindingsBatch {value=65535}
      executionRuntimeTemplateIndex {value=65535}
      instanceDataHandle {index=65535, source=None, stateHandle {index=65535}}
      instanceTemplateIndex {value=65535}
      outputBindingsBatch {value=65535}
      transitionTo {name=Choose CombatantCount, fallback=None, iD=#27, linkType=GotoState, stateHandle {index=65535}}
     }
    }
   }
  ]
  transitions []
 }
 Choose CombatantCount {
  id=#15
  type=STATE
  tasks [{
    id=#28
    description=Delay 0.01
    properties {
     iD=#29
     instance {_structType=&StateTreeDelayTaskInstanceData, duration=0.01}
     node {name=None, _structType=&StateTreeDelayTask}
    }
   }]
  transitions [
   {
    id=#30
    trigger=ON_STATE_COMPLETED
    conditions [{
      id=#31
      description="Get Variable [Int] - CombatantCount.Value == 3"
      properties {
       iD=#32
       instance {_structType=/Script/StateTreeModule.StateTreeCompareIntConditionInstanceData, left=0, right=3}
       node {
        name=None
        _structType=/Script/StateTreeModule.StateTreeCompareIntCondition
        bInvert=false
        deltaIndent=0
        evaluationMode=Evaluated
        operand=And
        operator=Equal
        bindingsBatch {value=65535}
        executionRuntimeTemplateIndex {value=65535}
        instanceDataHandle {index=65535, source=None, stateHandle {index=65535}}
        instanceTemplateIndex {value=65535}
        outputBindingsBatch {value=65535}
       }
      }
     }]
    state {id=#16, name=Stage_Wave_One_Small}
   }
   {
    id=#33
    trigger=ON_STATE_COMPLETED
    conditions [{
      id=#34
      description="Get Variable [Int] - CombatantCount.Value == 4"
      properties {
       iD=#35
       instance {_structType=&StateTreeCompareIntConditionInstanceData, left=0, right=4}
       node {
        name=None
        _structType=&StateTreeCompareIntCondition
        bInvert=false
        deltaIndent=0
        evaluationMode=Evaluated
        operand=And
        operator=Equal
        bindingsBatch {value=65535}
        executionRuntimeTemplateIndex {value=65535}
        instanceDataHandle {index=65535, source=None, stateHandle {index=65535}}
        instanceTemplateIndex {value=65535}
        outputBindingsBatch {value=65535}
       }
      }
     }]
    state {id=#17, name=Stage_Wave_One_Medium}
   }
  ]
 }
 Stage_Wave_One_Small {
  id=#16
  type=STATE
  tasks [
   {
    id=#36
    description=Spawn Wave One - 3 combatants from a weight-selected group in DT_Wave_Selection
    properties {
     iD=#37
     instance {
      _structType=/Script/EncountersRuntime.LWMEncounterDynamicStartStateTreeTaskInstanceData
      bAttachVolumeToSpawn=false
      bOverrideAnchorPosition=true
      bOverrideDelayBetweenEvents=false
      bOverrideEnvQuery=false
      bOverrideStageActorMaxCount=false
      bOverrideStageActorMaxSpawnedCount=false
      bPauseOnExit=false
      bSpecifyFaction=true
      bUseVolumeAsAnchor=true
      bWaitForEncounterDoneSpawning=false
      delayBetweenEventsMax=0
      delayBetweenEventsMin=0
      dynamicEmptyEncounter=/FortQuestEncounters/Lager/Lager_Encounter_Empty_QuestEncounter
      faction=/FortQuestEncounters/Factions/Spires/Faction_Spires
      instigator=null
      owningActor=null
      spawnAroundEnvironmentQuery=null
      spawnAroundEnvironmentQueryRunMode=AllMatching
      spawnAroundNavigationSourceOverride=""
      stageActorMaxCount=1
      stageActorMaxSpawnedCount=-1
      userActor=null
      actorTags {gameplayTags []}
      anchorPointTagQuery {autoDescription="", tokenStreamVersion=0, userDescription="", queryTokenStream [], tagDictionary []}
      dynamicEncounterEntries [
       {
        actorCountOverrideMax=1
        actorCountOverrideMin=1
        bCreateRandomSingleActorEvents=false
        bOverrideActorCount=false
        bUseEncounterTagQuery=true
        bUseEventTemplate=true
        debugName=""
        maxConcurrentSpawns=1
        noActorLimitEventCount=1
        spawnCount=1
        actorDescriptions []
        encounterTagQuery {
         autoDescription=" ANY( Athena.Quest.Scenario.Faction.Henchmen.Assault )"
         tokenStreamVersion=0
         userDescription=""
         queryTokenStream [0,1,1,1,0]
         tagDictionary [{tagName=Athena.Quest.Scenario.Faction.Henchmen.Assault}]
        }
        eventTemplate {
         activationGamePhaseStep=None
         activationMaxDelay=0
         activationMinDelay=0
         activationSafeZoneIndex=0
         bCanSpawnWithoutMatchingPlayerDistance=false
         bCheckForActorDensity=true
         bOnlyInNextSafeZone=false
         bOnlyInNextSafeZoneIfStormShrinking=false
         bOnlyInSafeZone=false
         bShouldProcessEventRequestImmediately=false
         deactivationMaxDelay=-1
         deactivationMinDelay=-1
         maxActorDensityOverride=-1
         maxDistanceToPlayer=-1
         minDistanceToPlayer=-1
         randomActivationChance=100
         weight=1
         calendarEventConditions {isActiveWithoutSeasonalManager=false, conditions []}
         canSpawnCondition {description=Empty, schemaClass=/Script/LagerRuntime.FortAthenaLivingWorldConditionSchema, editableConditions [], sharedDefinition {}}
         despawnCondition {
          bCanBeDespawned=false
          bCanDespawnInTheStorm=false
          bCanDespawnWhenInCombatWithPlayer=false
          canStartDespawningPhaseStep=None
          canStartDespawningSafeZonePhase=-1
          maxStimulusAgeToBeConsideredInCombat=30
          minDistanceToPlayer=-1
          timeAsCandidateBeforeDespawn=10
         }
         despawnConditions []
         eventDatas []
         eventInstanceMaxCount [-1]
         eventInstanceMaxSpawnedCount [-1]
         instantDeathMaxRespawnTime []
         instantDeathMinRespawnTime []
         isEnabled {registryType=None, value=1, curve {curveTable=null, rowName=None}}
         maxCount [1]
         maxRespawnTime [45]
         maxSpawnedCount [1]
         minRespawnTime [30]
         playlistTagQuery {autoDescription="", tokenStreamVersion=0, userDescription="", queryTokenStream [], tagDictionary []}
         tagDensityConditions []
        }
        pointProviderTagQueryOverride {autoDescription="", tokenStreamVersion=0, userDescription="", queryTokenStream [], tagDictionary []}
       }
       {
        actorCountOverrideMax=1
        actorCountOverrideMin=1
        bCreateRandomSingleActorEvents=false
        bOverrideActorCount=false
        bUseEncounterTagQuery=true
        bUseEventTemplate=true
        debugName=""
        maxConcurrentSpawns=1
        noActorLimitEventCount=1
        spawnCount=1
        actorDescriptions []
        encounterTagQuery {
         autoDescription=" ANY( Athena.Quest.Scenario.Faction.Henchmen.Assault )"
         tokenStreamVersion=0
         userDescription=""
         queryTokenStream [0,1,1,1,0]
         tagDictionary [{tagName=Athena.Quest.Scenario.Faction.Henchmen.Assault}]
        }
        eventTemplate {
         activationGamePhaseStep=None
         activationMaxDelay=0
         activationMinDelay=0
         activationSafeZoneIndex=0
         bCanSpawnWithoutMatchingPlayerDistance=false
         bCheckForActorDensity=true
         bOnlyInNextSafeZone=false
         bOnlyInNextSafeZoneIfStormShrinking=false
         bOnlyInSafeZone=false
         bShouldProcessEventRequestImmediately=false
         deactivationMaxDelay=-1
         deactivationMinDelay=-1
         maxActorDensityOverride=-1
         maxDistanceToPlayer=-1
         minDistanceToPlayer=-1
         randomActivationChance=100
         weight=1
         calendarEventConditions {isActiveWithoutSeasonalManager=false, conditions []}
         canSpawnCondition {description=Empty, schemaClass=&FortAthenaLivingWorldConditionSchema, editableConditions [], sharedDefinition {}}
         despawnCondition {
          bCanBeDespawned=false
          bCanDespawnInTheStorm=false
          bCanDespawnWhenInCombatWithPlayer=false
          canStartDespawningPhaseStep=None
          canStartDespawningSafeZonePhase=-1
          maxStimulusAgeToBeConsideredInCombat=30
          minDistanceToPlayer=-1
          timeAsCandidateBeforeDespawn=10
         }
         despawnConditions []
         eventDatas []
         eventInstanceMaxCount [-1]
         eventInstanceMaxSpawnedCount [-1]
         instantDeathMaxRespawnTime []
         instantDeathMinRespawnTime []
         isEnabled {registryType=None, value=1, curve {curveTable=null, rowName=None}}
         maxCount [1]
         maxRespawnTime [45]
         maxSpawnedCount [1]
         minRespawnTime [30]
         playlistTagQuery {autoDescription="", tokenStreamVersion=0, userDescription="", queryTokenStream [], tagDictionary []}
         tagDensityConditions []
        }
        pointProviderTagQueryOverride {autoDescription="", tokenStreamVersion=0, userDescription="", queryTokenStream [], tagDictionary []}
       }
       {
        actorCountOverrideMax=1
        actorCountOverrideMin=1
        bCreateRandomSingleActorEvents=false
        bOverrideActorCount=false
        bUseEncounterTagQuery=true
        bUseEventTemplate=true
        debugName=""
        maxConcurrentSpawns=1
        noActorLimitEventCount=1
        spawnCount=1
        actorDescriptions []
        encounterTagQuery {
         autoDescription=" ANY( Athena.Quest.Scenario.Faction.Henchmen.Assault )"
         tokenStreamVersion=0
         userDescription=""
         queryTokenStream [0,1,1,1,0]
         tagDictionary [{tagName=Athena.Quest.Scenario.Faction.Henchmen.Assault}]
        }
        eventTemplate {
         activationGamePhaseStep=None
         activationMaxDelay=0
         activationMinDelay=0
         activationSafeZoneIndex=0
         bCanSpawnWithoutMatchingPlayerDistance=false
         bCheckForActorDensity=true
         bOnlyInNextSafeZone=false
         bOnlyInNextSafeZoneIfStormShrinking=false
         bOnlyInSafeZone=false
         bShouldProcessEventRequestImmediately=false
         deactivationMaxDelay=-1
         deactivationMinDelay=-1
         maxActorDensityOverride=-1
         maxDistanceToPlayer=-1
         minDistanceToPlayer=-1
         randomActivationChance=100
         weight=1
         calendarEventConditions {isActiveWithoutSeasonalManager=false, conditions []}
         canSpawnCondition {description=Empty, schemaClass=&FortAthenaLivingWorldConditionSchema, editableConditions [], sharedDefinition {}}
         despawnCondition {
          bCanBeDespawned=false
          bCanDespawnInTheStorm=false
          bCanDespawnWhenInCombatWithPlayer=false
          canStartDespawningPhaseStep=None
          canStartDespawningSafeZonePhase=-1
          maxStimulusAgeToBeConsideredInCombat=30
          minDistanceToPlayer=-1
          timeAsCandidateBeforeDespawn=10
         }
         despawnConditions []
         eventDatas []
         eventInstanceMaxCount [-1]
         eventInstanceMaxSpawnedCount [-1]
         instantDeathMaxRespawnTime []
         instantDeathMinRespawnTime []
         isEnabled {registryType=None, value=1, curve {curveTable=null, rowName=None}}
         maxCount [1]
         maxRespawnTime [45]
         maxSpawnedCount [1]
         minRespawnTime [30]
         playlistTagQuery {autoDescription="", tokenStreamVersion=0, userDescription="", queryTokenStream [], tagDictionary []}
         tagDensityConditions []
        }
        pointProviderTagQueryOverride {autoDescription="", tokenStreamVersion=0, userDescription="", queryTokenStream [], tagDictionary []}
       }
      ]
      encounterIdentifierTags {gameplayTags [{tagName=QuestEncounter.Spire_Minor_Combat_1.AI.Goal.Assault.Wave.1}]}
      leashRadiusInner {registryType=None, value=1000, curve {curveTable=null, rowName=None}}
      leashRadiusOuter {registryType=None, value=2500, curve {curveTable=null, rowName=None}}
      pointProviderTagQuery {
       autoDescription=" ANY( QuestEncounter.Spire_Minor_Combat_1.Location.SpawnPoint.Wave )"
       tokenStreamVersion=0
       userDescription=Spawning will select a point for each combatant that matches this tag
       queryTokenStream [0,1,1,1,0]
       tagDictionary [{tagName=QuestEncounter.Spire_Minor_Combat_1.Location.SpawnPoint.Wave}]
      }
      spawnActions [{spawnActionClass=/SpiresGameplay/Scenarios/SpawnActions/BP_PlayVFX_SpawnFX.BP_PlayVFX_SpawnFX_C, spawnActionTag {tagName=None}}]
      spawnTimeout {registryType=None, value=-1, curve {curveTable=null, rowName=None}}
     }
     node {
      name=Spawn Wave One - 3 combatants from a weight-selected group in DT_Wave_Selection
      _structType=/Script/EncountersRuntime.LWMEncounterDynamicStartStateTreeTask
      bCanEditConsideredForCompletion=true
      bConsideredForCompletion=true
      bHasTaskCompletionDelegateDispatcher=false
      bTaskEnabled=true
      transitionHandlingPriority=Normal
      bindingsBatch {value=65535}
      executionRuntimeTemplateIndex {value=65535}
      instanceDataHandle {index=65535, source=None, stateHandle {index=65535}}
      instanceTemplateIndex {value=65535}
      outputBindingsBatch {value=65535}
     }
    }
   }
   {
    id=#38
    description=Grant Quest
    properties {
     iD=#39
     instance {
      _structType=/Script/EncountersRuntime.GrantQuestStateTreeTaskInstanceData
      grantType=GrantToPlayspace
      questToGrant=/SpiresGameplay/Scenarios/Spire_Minor/Spire_Minor_Combat_1/Quests/PlaySpace_EncounterQuest_Spire_Minor_Combat_1_01
      targetActor=null
      userActor=null
     }
     node {
      name=None
      _structType=/Script/EncountersRuntime.GrantQuestStateTreeTask
      bCanEditConsideredForCompletion=true
      bConsideredForCompletion=true
      bHasTaskCompletionDelegateDispatcher=false
      bTaskEnabled=false
      transitionHandlingPriority=Normal
      bindingsBatch {value=65535}
      executionRuntimeTemplateIndex {value=65535}
      instanceDataHandle {index=65535, source=None, stateHandle {index=65535}}
      instanceTemplateIndex {value=65535}
      outputBindingsBatch {value=65535}
     }
    }
   }
   {
    id=#40
    description=Delay 3.0
    properties {
     iD=#41
     instance {_structType=&StateTreeDelayTaskInstanceData, duration=3}
     node {name=None, _structType=&StateTreeDelayTask}
    }
   }
  ]
  transitions [{id=#42, trigger=ON_STATE_COMPLETED, state {id=#18, name=Set AI Faction}}]
 }
 Stage_Wave_One_Medium {
  id=#17
  type=STATE
  tasks [
   {
    id=#43
    description=Spawn Wave One - 4 combatants from a weight-selected group in DT_Wave_Selection
    properties {
     iD=#44
     instance {
      _structType=&LWMEncounterDynamicStartStateTreeTaskInstanceData
      bAttachVolumeToSpawn=false
      bOverrideAnchorPosition=true
      bOverrideDelayBetweenEvents=false
      bOverrideEnvQuery=false
      bOverrideStageActorMaxCount=false
      bOverrideStageActorMaxSpawnedCount=false
      bPauseOnExit=false
      bSpecifyFaction=true
      bUseVolumeAsAnchor=true
      bWaitForEncounterDoneSpawning=false
      delayBetweenEventsMax=0
      delayBetweenEventsMin=0
      dynamicEmptyEncounter=&Lager_Encounter_Empty_QuestEncounter
      faction=&Faction_Spires
      instigator=null
      owningActor=null
      spawnAroundEnvironmentQuery=null
      spawnAroundEnvironmentQueryRunMode=AllMatching
      spawnAroundNavigationSourceOverride=""
      stageActorMaxCount=1
      stageActorMaxSpawnedCount=-1
      userActor=null
      actorTags {gameplayTags []}
      anchorPointTagQuery {autoDescription="", tokenStreamVersion=0, userDescription="", queryTokenStream [], tagDictionary []}
      dynamicEncounterEntries [
       {
        actorCountOverrideMax=1
        actorCountOverrideMin=1
        bCreateRandomSingleActorEvents=false
        bOverrideActorCount=false
        bUseEncounterTagQuery=true
        bUseEventTemplate=true
        debugName=""
        maxConcurrentSpawns=1
        noActorLimitEventCount=1
        spawnCount=1
        actorDescriptions []
        encounterTagQuery {
         autoDescription=" ANY( Athena.Quest.Scenario.Faction.Henchmen.Assault )"
         tokenStreamVersion=0
         userDescription=""
         queryTokenStream [0,1,1,1,0]
         tagDictionary [{tagName=Athena.Quest.Scenario.Faction.Henchmen.Assault}]
        }
        eventTemplate {
         activationGamePhaseStep=None
         activationMaxDelay=0
         activationMinDelay=0
         activationSafeZoneIndex=0
         bCanSpawnWithoutMatchingPlayerDistance=false
         bCheckForActorDensity=true
         bOnlyInNextSafeZone=false
         bOnlyInNextSafeZoneIfStormShrinking=false
         bOnlyInSafeZone=false
         bShouldProcessEventRequestImmediately=false
         deactivationMaxDelay=-1
         deactivationMinDelay=-1
         maxActorDensityOverride=-1
         maxDistanceToPlayer=-1
         minDistanceToPlayer=-1
         randomActivationChance=100
         weight=1
         calendarEventConditions {isActiveWithoutSeasonalManager=false, conditions []}
         canSpawnCondition {description=Empty, schemaClass=&FortAthenaLivingWorldConditionSchema, editableConditions [], sharedDefinition {}}
         despawnCondition {
          bCanBeDespawned=false
          bCanDespawnInTheStorm=false
          bCanDespawnWhenInCombatWithPlayer=false
          canStartDespawningPhaseStep=None
          canStartDespawningSafeZonePhase=-1
          maxStimulusAgeToBeConsideredInCombat=30
          minDistanceToPlayer=-1
          timeAsCandidateBeforeDespawn=10
         }
         despawnConditions []
         eventDatas []
         eventInstanceMaxCount [-1]
         eventInstanceMaxSpawnedCount [-1]
         instantDeathMaxRespawnTime []
         instantDeathMinRespawnTime []
         isEnabled {registryType=None, value=1, curve {curveTable=null, rowName=None}}
         maxCount [1]
         maxRespawnTime [45]
         maxSpawnedCount [1]
         minRespawnTime [30]
         playlistTagQuery {autoDescription="", tokenStreamVersion=0, userDescription="", queryTokenStream [], tagDictionary []}
         tagDensityConditions []
        }
        pointProviderTagQueryOverride {autoDescription="", tokenStreamVersion=0, userDescription="", queryTokenStream [], tagDictionary []}
       }
       {
        actorCountOverrideMax=1
        actorCountOverrideMin=1
        bCreateRandomSingleActorEvents=false
        bOverrideActorCount=false
        bUseEncounterTagQuery=true
        bUseEventTemplate=true
        debugName=""
        maxConcurrentSpawns=1
        noActorLimitEventCount=1
        spawnCount=1
        actorDescriptions []
        encounterTagQuery {
         autoDescription=" ANY( Athena.Quest.Scenario.Faction.Henchmen.Assault )"
         tokenStreamVersion=0
         userDescription=""
         queryTokenStream [0,1,1,1,0]
         tagDictionary [{tagName=Athena.Quest.Scenario.Faction.Henchmen.Assault}]
        }
        eventTemplate {
         activationGamePhaseStep=None
         activationMaxDelay=0
         activationMinDelay=0
         activationSafeZoneIndex=0
         bCanSpawnWithoutMatchingPlayerDistance=false
         bCheckForActorDensity=true
         bOnlyInNextSafeZone=false
         bOnlyInNextSafeZoneIfStormShrinking=false
         bOnlyInSafeZone=false
         bShouldProcessEventRequestImmediately=false
         deactivationMaxDelay=-1
         deactivationMinDelay=-1
         maxActorDensityOverride=-1
         maxDistanceToPlayer=-1
         minDistanceToPlayer=-1
         randomActivationChance=100
         weight=1
         calendarEventConditions {isActiveWithoutSeasonalManager=false, conditions []}
         canSpawnCondition {description=Empty, schemaClass=&FortAthenaLivingWorldConditionSchema, editableConditions [], sharedDefinition {}}
         despawnCondition {
          bCanBeDespawned=false
          bCanDespawnInTheStorm=false
          bCanDespawnWhenInCombatWithPlayer=false
          canStartDespawningPhaseStep=None
          canStartDespawningSafeZonePhase=-1
          maxStimulusAgeToBeConsideredInCombat=30
          minDistanceToPlayer=-1
          timeAsCandidateBeforeDespawn=10
         }
         despawnConditions []
         eventDatas []
         eventInstanceMaxCount [-1]
         eventInstanceMaxSpawnedCount [-1]
         instantDeathMaxRespawnTime []
         instantDeathMinRespawnTime []
         isEnabled {registryType=None, value=1, curve {curveTable=null, rowName=None}}
         maxCount [1]
         maxRespawnTime [45]
         maxSpawnedCount [1]
         minRespawnTime [30]
         playlistTagQuery {autoDescription="", tokenStreamVersion=0, userDescription="", queryTokenStream [], tagDictionary []}
         tagDensityConditions []
        }
        pointProviderTagQueryOverride {autoDescription="", tokenStreamVersion=0, userDescription="", queryTokenStream [], tagDictionary []}
       }
       {
        actorCountOverrideMax=1
        actorCountOverrideMin=1
        bCreateRandomSingleActorEvents=false
        bOverrideActorCount=false
        bUseEncounterTagQuery=true
        bUseEventTemplate=true
        debugName=""
        maxConcurrentSpawns=1
        noActorLimitEventCount=1
        spawnCount=1
        actorDescriptions []
        encounterTagQuery {
         autoDescription=" ANY( Athena.Quest.Scenario.Faction.Henchmen.Assault )"
         tokenStreamVersion=0
         userDescription=""
         queryTokenStream [0,1,1,1,0]
         tagDictionary [{tagName=Athena.Quest.Scenario.Faction.Henchmen.Assault}]
        }
        eventTemplate {
         activationGamePhaseStep=None
         activationMaxDelay=0
         activationMinDelay=0
         activationSafeZoneIndex=0
         bCanSpawnWithoutMatchingPlayerDistance=false
         bCheckForActorDensity=true
         bOnlyInNextSafeZone=false
         bOnlyInNextSafeZoneIfStormShrinking=false
         bOnlyInSafeZone=false
         bShouldProcessEventRequestImmediately=false
         deactivationMaxDelay=-1
         deactivationMinDelay=-1
         maxActorDensityOverride=-1
         maxDistanceToPlayer=-1
         minDistanceToPlayer=-1
         randomActivationChance=100
         weight=1
         calendarEventConditions {isActiveWithoutSeasonalManager=false, conditions []}
         canSpawnCondition {description=Empty, schemaClass=&FortAthenaLivingWorldConditionSchema, editableConditions [], sharedDefinition {}}
         despawnCondition {
          bCanBeDespawned=false
          bCanDespawnInTheStorm=false
          bCanDespawnWhenInCombatWithPlayer=false
          canStartDespawningPhaseStep=None
          canStartDespawningSafeZonePhase=-1
          maxStimulusAgeToBeConsideredInCombat=30
          minDistanceToPlayer=-1
          timeAsCandidateBeforeDespawn=10
         }
         despawnConditions []
         eventDatas []
         eventInstanceMaxCount [-1]
         eventInstanceMaxSpawnedCount [-1]
         instantDeathMaxRespawnTime []
         instantDeathMinRespawnTime []
         isEnabled {registryType=None, value=1, curve {curveTable=null, rowName=None}}
         maxCount [1]
         maxRespawnTime [45]
         maxSpawnedCount [1]
         minRespawnTime [30]
         playlistTagQuery {autoDescription="", tokenStreamVersion=0, userDescription="", queryTokenStream [], tagDictionary []}
         tagDensityConditions []
        }
        pointProviderTagQueryOverride {autoDescription="", tokenStreamVersion=0, userDescription="", queryTokenStream [], tagDictionary []}
       }
       {
        actorCountOverrideMax=1
        actorCountOverrideMin=1
        bCreateRandomSingleActorEvents=false
        bOverrideActorCount=false
        bUseEncounterTagQuery=true
        bUseEventTemplate=true
        debugName=""
        maxConcurrentSpawns=1
        noActorLimitEventCount=1
        spawnCount=1
        actorDescriptions []
        encounterTagQuery {
         autoDescription=" ANY( Athena.Quest.Scenario.Faction.Henchmen.Assault )"
         tokenStreamVersion=0
         userDescription=""
         queryTokenStream [0,1,1,1,0]
         tagDictionary [{tagName=Athena.Quest.Scenario.Faction.Henchmen.Assault}]
        }
        eventTemplate {
         activationGamePhaseStep=None
         activationMaxDelay=0
         activationMinDelay=0
         activationSafeZoneIndex=0
         bCanSpawnWithoutMatchingPlayerDistance=false
         bCheckForActorDensity=true
         bOnlyInNextSafeZone=false
         bOnlyInNextSafeZoneIfStormShrinking=false
         bOnlyInSafeZone=false
         bShouldProcessEventRequestImmediately=false
         deactivationMaxDelay=-1
         deactivationMinDelay=-1
         maxActorDensityOverride=-1
         maxDistanceToPlayer=-1
         minDistanceToPlayer=-1
         randomActivationChance=100
         weight=1
         calendarEventConditions {isActiveWithoutSeasonalManager=false, conditions []}
         canSpawnCondition {description=Empty, schemaClass=&FortAthenaLivingWorldConditionSchema, editableConditions [], sharedDefinition {}}
         despawnCondition {
          bCanBeDespawned=false
          bCanDespawnInTheStorm=false
          bCanDespawnWhenInCombatWithPlayer=false
          canStartDespawningPhaseStep=None
          canStartDespawningSafeZonePhase=-1
          maxStimulusAgeToBeConsideredInCombat=30
          minDistanceToPlayer=-1
          timeAsCandidateBeforeDespawn=10
         }
         despawnConditions []
         eventDatas []
         eventInstanceMaxCount [-1]
         eventInstanceMaxSpawnedCount [-1]
         instantDeathMaxRespawnTime []
         instantDeathMinRespawnTime []
         isEnabled {registryType=None, value=1, curve {curveTable=null, rowName=None}}
         maxCount [1]
         maxRespawnTime [45]
         maxSpawnedCount [1]
         minRespawnTime [30]
         playlistTagQuery {autoDescription="", tokenStreamVersion=0, userDescription="", queryTokenStream [], tagDictionary []}
         tagDensityConditions []
        }
        pointProviderTagQueryOverride {autoDescription="", tokenStreamVersion=0, userDescription="", queryTokenStream [], tagDictionary []}
       }
      ]
      encounterIdentifierTags {gameplayTags [{tagName=QuestEncounter.Spire_Minor_Combat_1.AI.Goal.Assault.Wave.1}]}
      leashRadiusInner {registryType=None, value=1000, curve {curveTable=null, rowName=None}}
      leashRadiusOuter {registryType=None, value=2500, curve {curveTable=null, rowName=None}}
      pointProviderTagQuery {
       autoDescription=" ANY( QuestEncounter.Spire_Minor_Combat_1.Location.SpawnPoint.Wave )"
       tokenStreamVersion=0
       userDescription=Spawning will select a point for each combatant that matches this tag
       queryTokenStream [0,1,1,1,0]
       tagDictionary [{tagName=QuestEncounter.Spire_Minor_Combat_1.Location.SpawnPoint.Wave}]
      }
      spawnActions [{spawnActionClass=&BP_PlayVFX_SpawnFX_C, spawnActionTag {tagName=None}}]
      spawnTimeout {registryType=None, value=-1, curve {curveTable=null, rowName=None}}
     }
     node {
      name=Spawn Wave One - 4 combatants from a weight-selected group in DT_Wave_Selection
      _structType=&LWMEncounterDynamicStartStateTreeTask
      bCanEditConsideredForCompletion=true
      bConsideredForCompletion=true
      bHasTaskCompletionDelegateDispatcher=false
      bTaskEnabled=true
      transitionHandlingPriority=Normal
      bindingsBatch {value=65535}
      executionRuntimeTemplateIndex {value=65535}
      instanceDataHandle {index=65535, source=None, stateHandle {index=65535}}
      instanceTemplateIndex {value=65535}
      outputBindingsBatch {value=65535}
     }
    }
   }
   {
    id=#45
    description=Grant Quest
    properties {
     iD=#46
     instance {
      _structType=&GrantQuestStateTreeTaskInstanceData
      grantType=GrantToPlayspace
      questToGrant=/SpiresGameplay/Scenarios/Spire_Minor/Spire_Minor_Combat_1/Quests/PlaySpace_EncounterQuest_Spire_Minor_Combat_1_00
      targetActor=null
      userActor=null
     }
     node {
      name=None
      _structType=&GrantQuestStateTreeTask
      bCanEditConsideredForCompletion=true
      bConsideredForCompletion=true
      bHasTaskCompletionDelegateDispatcher=false
      bTaskEnabled=false
      transitionHandlingPriority=Normal
      bindingsBatch {value=65535}
      executionRuntimeTemplateIndex {value=65535}
      instanceDataHandle {index=65535, source=None, stateHandle {index=65535}}
      instanceTemplateIndex {value=65535}
      outputBindingsBatch {value=65535}
     }
    }
   }
   {
    id=#47
    description=Delay 3.0
    properties {
     iD=#48
     instance {_structType=&StateTreeDelayTaskInstanceData, duration=3}
     node {name=None, _structType=&StateTreeDelayTask}
    }
   }
  ]
  transitions [{id=#49, trigger=ON_STATE_COMPLETED, state {id=#18, name=Set AI Faction}}]
 }
 Set AI Faction {
  id=#18
  type=STATE
  tasks [
   {
    id=#50
    description="LWM Encounter [Modify Actor Faction]"
    properties {
     iD=#51
     instance {
      _structType=/Script/EncountersRuntime.LWMEncounterModifyActorFactionStateTreeTaskInstanceData
      modificationAction=ExclusiveSet
      userActor=null
      encounterQuery {autoDescription="", tokenStreamVersion=0, userDescription="", queryTokenStream [], tagDictionary []}
      factionTag {tagName=Athena.Faction.CherrySmoke.Hostile}
     }
     node {
      name=None
      _structType=/Script/EncountersRuntime.LWMEncounterModifyActorFactionStateTreeTask
      bCanEditConsideredForCompletion=true
      bConsideredForCompletion=true
      bHasTaskCompletionDelegateDispatcher=false
      bTaskEnabled=true
      transitionHandlingPriority=Normal
      bindingsBatch {value=65535}
      executionRuntimeTemplateIndex {value=65535}
      instanceDataHandle {index=65535, source=None, stateHandle {index=65535}}
      instanceTemplateIndex {value=65535}
      outputBindingsBatch {value=65535}
     }
    }
   }
   {
    id=#52
    description="LWM Encounter [Remaining Count Transition]"
    properties {
     iD=#53
     instance {
      _structType=/Script/EncountersRuntime.LWMEncounterRemainingCountTransitionStateTreeTaskInstanceData
      bOnlyTriggerOnce=true
      bShouldTransition=true
      bTriggered=false
      userActor=null
      value=0
      customVerb {amount=1, customVerbTag {tagName=None}, subjects [], values []}
      encounterQuery {
       autoDescription=""
       tokenStreamVersion=0
       userDescription=""
       queryTokenStream [0,1,1,1,0]
       tagDictionary [{tagName=QuestEncounter.Spire_Minor_Combat_1.AI.Goal.Assault.Wave.1}]
      }
      onTriggered {iD=#26}
     }
     node {
      name=None
      _structType=/Script/EncountersRuntime.LWMEncounterRemainingCountTransitionStateTreeTask
      bCanEditConsideredForCompletion=true
      bConsideredForCompletion=true
      bHasTaskCompletionDelegateDispatcher=false
      bTaskEnabled=false
      transitionHandlingPriority=Normal
      bindingsBatch {value=65535}
      executionRuntimeTemplateIndex {value=65535}
      instanceDataHandle {index=65535, source=None, stateHandle {index=65535}}
      instanceTemplateIndex {value=65535}
      outputBindingsBatch {value=65535}
      transitionTo {name=Spire_Vulnerable, fallback=None, iD=#54, linkType=GotoState, stateHandle {index=65535}}
     }
    }
   }
  ]
  transitions []
 }
 Wait for Minor Spire destroyed {
  id=#19
  type=STATE
  tasks [
   {
    id=#55
    description=Grant Reactions - Spire vulnerable
    properties {
     iD=#56
     instance {
      _structType=&GrantReactionStateTreeTaskInstanceData
      grantType=GrantToTargetActor
      targetActor=null
      userActor=null
      reactionEntries [{rowValue {
         bUseDataRegistryId=false
         dataRegistryId=""
         dataTableRow {
          dataTable=&DT_QuestScenario_Spire_Minor_Combat_1_Reactions
          rowName=SpireVulnerable
         }
        }}]
      reactions []
     }
     node {
      name=Grant Reactions - Spire vulnerable
      _structType=&GrantReactionStateTreeTask
      bCanEditConsideredForCompletion=true
      bConsideredForCompletion=true
      bHasTaskCompletionDelegateDispatcher=false
      bTaskEnabled=false
      transitionHandlingPriority=Normal
      bindingsBatch {value=65535}
      executionRuntimeTemplateIndex {value=65535}
      instanceDataHandle {index=65535, source=None, stateHandle {index=65535}}
      instanceTemplateIndex {value=65535}
      outputBindingsBatch {value=65535}
     }
    }
   }
   {
    id=#57
    description=Grant Quest
    properties {
     iD=#58
     instance {
      _structType=&GrantQuestStateTreeTaskInstanceData
      grantType=GrantToPlayspace
      questToGrant=/SpiresGameplay/Scenarios/Spire_Minor/Spire_Minor_Combat_1/Quests/PlaySpace_EncounterQuest_Spire_Minor_Combat_1_03
      targetActor=null
      userActor=null
     }
     node {
      name=None
      _structType=&GrantQuestStateTreeTask
      bCanEditConsideredForCompletion=true
      bConsideredForCompletion=true
      bHasTaskCompletionDelegateDispatcher=false
      bTaskEnabled=false
      transitionHandlingPriority=Normal
      bindingsBatch {value=65535}
      executionRuntimeTemplateIndex {value=65535}
      instanceDataHandle {index=65535, source=None, stateHandle {index=65535}}
      instanceTemplateIndex {value=65535}
      outputBindingsBatch {value=65535}
     }
    }
   }
   {
    id=#59
    description=Scenario Verb Transition
    properties {
     iD=#60
     instance {
      _structType=/Script/EncountersRuntime.EncounterVerbTransitionStateTreeTaskInstanceData
      bIsTriggered=false
      bUseAdvancedObjectives=false
      objectiveValueThreshold=1
      userActor=null
      objectiveVerbs [{
        _structType=/Script/FortniteGame.FortObjectiveFilter_BuildingWasDestroyed
        author {
         bCanGainProgressForThisObjective=true
         tagQuery {
          autoDescription=" ANY( QuestEncounter.Spire_Minor_Random.Encounter.Target.Spire )"
          tokenStreamVersion=0
          userDescription=""
          queryTokenStream [0,1,1,1,0]
          tagDictionary [{tagName=QuestEncounter.Spire_Minor_Random.Encounter.Target.Spire}]
         }
        }
        context {tagQuery {autoDescription="", tokenStreamVersion=0, userDescription="", queryTokenStream [], tagDictionary []}}
        damageSource {
         bCanGainProgressForThisObjective=true
         tagQuery {autoDescription="", tokenStreamVersion=0, userDescription="", queryTokenStream [], tagDictionary []}
        }
       }]
      objectives []
      onTriggered {iD=#26}
     }
     node {
      name=None
      _structType=/Script/EncountersRuntime.EncounterVerbTransitionStateTreeTask
      bCanEditConsideredForCompletion=true
      bConsideredForCompletion=true
      bHasTaskCompletionDelegateDispatcher=false
      bTaskEnabled=true
      transitionHandlingPriority=Normal
      bindingsBatch {value=65535}
      executionRuntimeTemplateIndex {value=65535}
      instanceDataHandle {index=65535, source=None, stateHandle {index=65535}}
      instanceTemplateIndex {value=65535}
      outputBindingsBatch {value=65535}
      transitionTo {name=Complete, fallback=None, iD=#61, linkType=GotoState, stateHandle {index=65535}}
     }
    }
   }
  ]
  transitions []
 }
 Stage_Reward {
  id=#20
  type=STATE
  tasks [
   {
    id=#62
    description=Get Anchor - Reward
    properties {
     iD=#63
     instance {
      _structType=/Script/EncountersRuntime.GetAnchorStateTreeTaskInstanceData
      anchorActor=null
      userActor=null
      anchorQuery {
       autoDescription=" ANY( QuestEncounter.Spire_Minor_Combat_1.Location.Reward.1 )"
       tokenStreamVersion=0
       userDescription=""
       queryTokenStream [0,1,1,1,0]
       tagDictionary [{tagName=QuestEncounter.Spire_Minor_Combat_1.Location.Reward.1}]
      }
     }
     node {
      name=Get Anchor - Reward
      _structType=/Script/EncountersRuntime.GetAnchorStateTreeTask
      bCanEditConsideredForCompletion=true
      bConsideredForCompletion=true
      bHasTaskCompletionDelegateDispatcher=false
      bTaskEnabled=false
      transitionHandlingPriority=Normal
      bindingsBatch {value=65535}
      executionRuntimeTemplateIndex {value=65535}
      instanceDataHandle {index=65535, source=None, stateHandle {index=65535}}
      instanceTemplateIndex {value=65535}
      outputBindingsBatch {value=65535}
     }
    }
   }
   {
    id=#64
    description=Spawn Pickup
    properties {
     iD=#65
     instance {
      _structType=/Script/FortniteGame.SpawnPickupStateTreeTaskInstanceData
      instigatorActor=null
      itemDefinition=null
      lootTierName=Loot_GoldBars_XL
      spawnedPickup=null
      targetActor=null
      userActor=null
      flingConeAngle {registryType=None, value=0, curve {curveTable=null, rowName=None}}
      flingDirection=<0 0 1>
      flingMagnitude {registryType=None, value=0, curve {curveTable=null, rowName=None}}
      spawnOffset=<0 0 0>
      spawnedPickupActors []
     }
     node {
      name=None
      _structType=/Script/FortniteGame.SpawnPickupStateTreeTask
      bCanEditConsideredForCompletion=true
      bConsideredForCompletion=true
      bHasTaskCompletionDelegateDispatcher=false
      bTaskEnabled=false
      transitionHandlingPriority=Normal
      bindingsBatch {value=65535}
      executionRuntimeTemplateIndex {value=65535}
      instanceDataHandle {index=65535, source=None, stateHandle {index=65535}}
      instanceTemplateIndex {value=65535}
      outputBindingsBatch {value=65535}
     }
    }
   }
   {
    id=#66
    description=Delay 1.0
    properties {
     iD=#67
     instance {_structType=&StateTreeDelayTaskInstanceData, duration=1}
     node {name=None, _structType=&StateTreeDelayTask}
    }
   }
  ]
  transitions [{id=#68, trigger=ON_STATE_COMPLETED, state {id=#21, name=Complete}}]
 }
 Complete {
  id=#21
  type=STATE
  tasks [
   {
    id=#69
    description=Send Custom Verb
    properties {
     iD=#70
     instance {
      _structType=/Script/EncountersRuntime.EncounterSendCustomVerbStateTreeTaskInstanceData
      bSendVerbForEachPlayerAuthor=false
      userActor=null
      customVerb {amount=1, customVerbTag {tagName=CustomVerb.Encounter.Completed}, subjects [], values []}
     }
     node {
      name=None
      _structType=/Script/EncountersRuntime.EncounterSendCustomVerbStateTreeTask
      bCanEditConsideredForCompletion=true
      bConsideredForCompletion=true
      bHasTaskCompletionDelegateDispatcher=false
      bTaskEnabled=true
      transitionHandlingPriority=Normal
      bindingsBatch {value=65535}
      executionRuntimeTemplateIndex {value=65535}
      instanceDataHandle {index=65535, source=None, stateHandle {index=65535}}
      instanceTemplateIndex {value=65535}
      outputBindingsBatch {value=65535}
     }
    }
   }
   {
    id=#71
    description=Grant Reactions
    properties {
     iD=#72
     instance {
      _structType=&GrantReactionStateTreeTaskInstanceData
      grantType=GrantToParticipants
      targetActor=null
      userActor=null
      reactionEntries []
      reactions [
       {
        _structType=/Script/FortniteGame.FortEventReaction_PlayVFX
        bForceVFXAtTargetLocation=false
        vFX=/Game/Effects/Fort_Effects/Gameplay/TimeTrial/NS_S29_MythicLabors_FootRace_Success
        effectLocationOffset=<0 0 0>
        effectScale=<1 1 1>
       }
       {
        _structType=/Script/FortniteGame.FortEventReaction_PlayAudio
        audio=/Game/Athena/Sounds/Quests/TimedTrial/Athena_Quest_TimeTrial_Success_Cue
        bForceSoundAtTargetLocation=false
        pitchMultiplier=1
        volumeMultiplier=1
        soundLocationOffset=<0 0 0>
       }
      ]
     }
     node {
      name=None
      _structType=&GrantReactionStateTreeTask
      bCanEditConsideredForCompletion=true
      bConsideredForCompletion=true
      bHasTaskCompletionDelegateDispatcher=false
      bTaskEnabled=true
      transitionHandlingPriority=Normal
      bindingsBatch {value=65535}
      executionRuntimeTemplateIndex {value=65535}
      instanceDataHandle {index=65535, source=None, stateHandle {index=65535}}
      instanceTemplateIndex {value=65535}
      outputBindingsBatch {value=65535}
     }
    }
   }
  ]
  transitions [{id=#73, trigger=ON_STATE_COMPLETED, state {id=#74, name=Root, linkType=SUCCEEDED}}]
 }
]
